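# Physics asset for the MoonFlax SUV vehicle body cosmetic, named as the BR variant.
# It defines a single body on the root bone, with no collision shapes or property overrides shown.
# It has no constraints, so it sets up no jointed or articulated parts.
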PhysicsAsset /VehicleCosmetics/Bodies/SUV_MoonFlax/Meshes/Body_MoonFlaxSUV_PhysicsBR
bodies {root []}
constraints []
