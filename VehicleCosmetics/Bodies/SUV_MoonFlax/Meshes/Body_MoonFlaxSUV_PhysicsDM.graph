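# Physics asset for the MoonFlax SUV body mesh in the vehicle cosmetics content.
# It approximates the whole vehicle body with a single box shape on the root bone, raised and offset slightly along X.
# It defines no other bodies and no constraints, so it has no articulated parts.
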PhysicsAsset /VehicleCosmetics/Bodies/SUV_MoonFlax/Meshes/Body_MoonFlaxSUV_PhysicsDM
bodies {root [{shapeType=Box, extentX=444, extentY=216, extentZ=132, shapeName=root_box, center=<19.5077 -0.0002 103.9446>}]}
constraints []
